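# Niagara system for the silver Raider backpack cosmetic that produces intermittent bright sprite flashes on the accessory.
# Two looping emitters each burst one randomly rotated sprite per loop, with randomized durations and delays so timing varies.
# Flash_Cylinder spawns on a cylinder's surface, while Flash_Face spawns within a thinner, narrower cylinder volume and repeats more often.
# Curves pulse color, alpha and size up and down over each particle's life, with a small camera offset.
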NiagaraSystem /BRCosmetics/Accessories/FORT_Backpacks/Backpack_F_MED_Raider_Silver/FX/NS_Raider_Silver_Flash
dependencies {usedDataInterfaces [&NiagaraDataInterfaceCurve], usedDynamicInputs [], usedModules [&SystemState]}
emitters {
 Flash_Cylinder {
  emitterUpdateScript [
   &EmitterState(
    Life Cycle Mode=::NewEnumerator1
    Loop Delay=&UniformRangedFloat(Minimum=1.2, Maximum=2)
    Loop Duration=&UniformRangedFloat(Minimum=1.2, Maximum=2)
    Recalculate Duration Each Loop=true
    UseLoopDelay=true
   )
   &SpawnBurst_Instantaneous(Spawn Count=1, Use Spawn Probability=false)
  ]
  particleSpawnScript [
   /Niagara/Modules/Spawn/Initialization/InitializeParticle(
    Color=<rgba 5 5 5 1>
    Lifetime=&UniformRangedFloat(Minimum=0.7, Maximum=1)
    Module.Write Color=true
    Module.Write Lifetime=true
    Module.Write Mass=true
    Module.Write Material Random=true
    Module.Write Position=true
    Module.Write Scale=false
    Module.Write SpriteRotation=true
    Module.Write SpriteSize=true
    Sprite Rotation=&UniformRangedFloat(Minimum=0, Maximum=360)
    Sprite Size=&Vector2DFromFloat(Value=&UniformRangedFloat(Minimum=10, Maximum=15))
   )
   &CylinderLocation(
    Cylinder Height=2.5
    Cylinder Radius=8
    MaskSpawnGroup=false
    Module.Enable Lathe Profile=false
    Module.Override Randomness=false
    Module.Surface Only=true
    Offset=<0 -11 1>
    Orientation Axis:ENiagaraOrientationAxis=::YAxis
    Surface Only Band Thickness=1
   )
  ]
  particleUpdateScript [
   &ScaleColor(
    Scale Alpha=&FloatFromCurve(FloatCurve=DataInterface() {propertyValues {Curve {keys [<at 0 0>,<at 0.5016 0.9767>,<at 1 0>]}}})
    Scale RGB=&VectorFromFloat(Value=&FloatFromCurve(FloatCurve=DataInterface() {propertyValues {Curve {keys [<at 0 0>,<at 0.5017 1.0042>,<at 1 0>]}}}))
    ScaleA=true
    ScaleRGB=true
   )
   &ParticleState()
   &SolveForcesAndVelocity(Clamp Velocity=false, Limit Acceleration=false)
   &ScaleSpriteSize(Scale Factor=&Vector2DFromFloat(Value=&FloatFromCurve(
      FloatCurve=DataInterface() {propertyValues {Curve {keys [
          <at 0 0 RCIM_Cubic>
          {arriveTangent=1.9892, interpMode=RCIM_Cubic, leaveTangent=1.9892, time=0.335, value=0.2056}
          {arriveTangent=0.2473, interpMode=RCIM_Cubic, leaveTangent=0.2473, time=0.5002, value=0.995}
          {arriveTangent=-1.9908, interpMode=RCIM_Cubic, leaveTangent=-1.9908, time=0.6669, value=0.2877}
          <at 1 0>
         ]}}}
      Scale Curve=1.5
     )))
   &CameraOffset(Camera Offset Amount=5)
  ]
  rendererClasses [&NiagaraSpriteRendererProperties]
  renderers [{rendererClass=&NiagaraSpriteRendererProperties}]
 }
 Flash_Face {
  emitterUpdateScript [
   &EmitterState(
    Life Cycle Mode=::NewEnumerator1
    Loop Delay=&UniformRangedFloat(Minimum=0, Maximum=1)
    Loop Duration=&UniformRangedFloat(Minimum=0.5, Maximum=1)
    Recalculate Duration Each Loop=true
    UseLoopDelay=true
   )
   &SpawnBurst_Instantaneous(Spawn Count=1, Use Spawn Probability=false)
   &SpawnRate(SpawnRate=500, Use Spawn Probability=false) {enabled=false}
  ]
  particleSpawnScript [
   &InitializeParticle(
    Color=<rgba 5 5 5 1>
    Lifetime=&UniformRangedFloat(Minimum=0.7, Maximum=1)
    Module.Write Color=true
    Module.Write Lifetime=true
    Module.Write Mass=true
    Module.Write Material Random=true
    Module.Write Position=true
    Module.Write Scale=false
    Module.Write SpriteRotation=true
    Module.Write SpriteSize=true
    Sprite Rotation=&UniformRangedFloat(Minimum=0, Maximum=360)
    Sprite Size=&Vector2DFromFloat(Value=&UniformRangedFloat(Minimum=13, Maximum=15))
   )
   &SphereLocation(
    Hemisphere Y=true
    MaskSpawnGroup=false
    Offset=<0 -10 0>
    Override Seed=false
    Sphere Orientation Axis:Vector3f=<-1 0 0>
    Sphere Radius=7
    Surface Only=false
    UseNonUniformScale=false
    UseOffset=true
    UseSphereOrientation=true
   ) {enabled=false}
   &CylinderLocation(
    Cylinder Height=1
    Cylinder Radius=6
    Lathe Profile=DataInterface() {propertyValues {Curve {keys [<at 0 0.451 RCIM_Cubic>, {arriveTangent=0.049, interpMode=RCIM_Cubic, leaveTangent=0.049, time=0.5, value=1}, <at 1 0.5 RCIM_Cubic>]}}}
    MaskSpawnGroup=false
    Module.Enable Lathe Profile=false
    Module.Override Randomness=false
    Module.Surface Only=false
    Offset=<0 -13 1>
    Orientation Axis:ENiagaraOrientationAxis=::YAxis
    Surface Only Band Thickness=1
   )
  ]
  particleUpdateScript [
   &ScaleColor(
    Scale Alpha=&FloatFromCurve(FloatCurve=DataInterface() {propertyValues {Curve {keys [<at 0 0>,<at 0.5016 0.9767>,<at 1 0>]}}})
    Scale RGB=&VectorFromFloat(Value=&FloatFromCurve(FloatCurve=DataInterface() {propertyValues {Curve {keys [<at 0 0>,<at 0.5017 1.0042>,<at 1 0>]}}}))
    ScaleA=true
    ScaleRGB=true
   )
   &ParticleState()
   &SolveForcesAndVelocity(Clamp Velocity=false, Limit Acceleration=false)
   &ScaleSpriteSize(Scale Factor=&Vector2DFromFloat(Value=&FloatFromCurve(
      FloatCurve=DataInterface() {propertyValues {Curve {keys [
          <at 0 0 RCIM_Cubic>
          {arriveTangent=1.9892, interpMode=RCIM_Cubic, leaveTangent=1.9892, time=0.335, value=0.2056}
          {arriveTangent=0.2473, interpMode=RCIM_Cubic, leaveTangent=0.2473, time=0.5002, value=0.995}
          {arriveTangent=-1.9908, interpMode=RCIM_Cubic, leaveTangent=-1.9908, time=0.6669, value=0.2877}
          <at 1 0>
         ]}}}
      Scale Curve=1.5
     )))
   &CameraOffset(Camera Offset Amount=5)
  ]
  rendererClasses [&NiagaraSpriteRendererProperties]
  renderers [{rendererClass=&NiagaraSpriteRendererProperties}]
 }
}
summary {
 systemName=NS_Raider_Silver_Flash
 emitters [
  {emitterName=Flash_Cylinder, rendererClasses [&NiagaraSpriteRendererProperties]}
  {emitterName=Flash_Face, rendererClasses [&NiagaraSpriteRendererProperties]}
 ]
 userVariables []
}
