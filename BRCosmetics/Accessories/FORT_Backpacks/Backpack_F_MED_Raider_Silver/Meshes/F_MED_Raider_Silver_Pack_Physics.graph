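# Physics asset for the F_MED_Raider_Silver backpack cosmetic, defining collision bodies and joints for its hanging grenades, grenade pins and zipper.
# All bodies use capsules, with a large one for the pack center and two per grenade pin.
# Left and right grenades attach tightly to the pack center, each pin swings more loosely from its grenade, and the zipper swings from the center with little twist.
# The root and pack center have no constraint between them.
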PhysicsAsset /BRCosmetics/Accessories/FORT_Backpacks/Backpack_F_MED_Raider_Silver/Meshes/F_MED_Raider_Silver_Pack_Physics
bodies {
 dyn_L_grenade_jnt [{shapeType=Capsule, length=3.75, radius=2.75, center=<3.5 0 0>}]
 dyn_L_grenade_pin_2_base_jnt [
  {shapeType=Capsule, length=2.3541, radius=0.6459, center=<2.2699 -0.1193 0.298>, rotation=<rot 67.3369 -106.4162 -105.2096>}
  {
   shapeType=Capsule
   length=4.1893
   radius=0.636
   shapeName=dyn_L_grenade_pin_2_base_jnt_capsule
   center=<2.4858 -0.0891 0.5536>
   rotation=<rot 10.025 -4.7032 -116.2763>
  }
 ]
 dyn_R_grenade_jnt [{shapeType=Capsule, length=3.75, radius=2.75, center=<3.5 0 0>}]
 dyn_R_grenade_pin_2_base_jnt [
  {shapeType=Capsule, length=2.4283, radius=0.5717, center=<2.2416 0.1595 0.4225>, rotation=<rot 67.8808 111.6994 110.1937>}
  {
   shapeType=Capsule
   length=4.6914
   radius=0.6742
   shapeName=dyn_R_grenade_pin_2_base_jnt_capsule
   center=<2.4371 -0.0073 0.2768>
   rotation=<rot -0.6833 1.3998 -63.9951>
  }
 ]
 dyn_backpack_center_jnt [{shapeType=Capsule, length=7, radius=9.25, center=<0.0001 0.8975 1.4256>, rotation=<rot 0 0 90>}]
 dyn_zipper_base_jnt [{shapeType=Capsule, length=0.8099, radius=0.5, center=<2.1551 0 0.0852>, rotation=<rot -90 0 90.0002>}]
 root [{shapeType=Capsule, length=1, radius=1}]
}
constraints [
 {bone1Name=dyn_R_grenade_jnt, bone2Name=dyn_backpack_center_jnt, swing1LimitDegrees=5, swing2LimitDegrees=5, twistLimitDegrees=5}
 {bone1Name=dyn_R_grenade_pin_2_base_jnt, bone2Name=dyn_R_grenade_jnt, swing1LimitDegrees=30, swing2LimitDegrees=30, twistLimitDegrees=15}
 {bone1Name=dyn_zipper_base_jnt, bone2Name=dyn_backpack_center_jnt, swing1LimitDegrees=30, swing2LimitDegrees=30, twistLimitDegrees=2}
 {bone1Name=dyn_L_grenade_jnt, bone2Name=dyn_backpack_center_jnt, swing1LimitDegrees=5, swing2LimitDegrees=5, twistLimitDegrees=5}
 {bone1Name=dyn_L_grenade_pin_2_base_jnt, bone2Name=dyn_L_grenade_jnt, swing1LimitDegrees=30, swing2LimitDegrees=30, twistLimitDegrees=15}
]
