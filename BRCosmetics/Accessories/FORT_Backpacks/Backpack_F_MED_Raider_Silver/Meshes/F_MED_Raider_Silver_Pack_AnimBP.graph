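# Animation Blueprint for the Raider Silver backpack, adding physics-driven secondary motion that adapts to front-end and skydiving states.
# A RigidBody node simulates in root-bone space, with event-driven external force and acceleration scale.
# The update event dampens acceleration in the front end and weakens downward force while skydiving before parachute opening.
# Dive mode blends in additive offsets on the left and right grenade bones through an interpolated alpha.
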AnimBlueprint /BRCosmetics/Accessories/FORT_Backpacks/Backpack_F_MED_Raider_Silver/Meshes/F_MED_Raider_Silver_Pack_AnimBP
target_skeleton=&/F_MED_Raider_Silver_Pack_Skeleton
graphs [
 AnimGraph {nodes [
   n1 = RigidBody!(
    ComponentPose=Local To Component(LocalPose=Input Pose())
    Alpha
    OverrideWorldGravity:Vector="0,0,-980"
    ExternalForce=Get gravity_override() -> (gravity_override:Vector)
    ComponentLinearAccScale:Vector=Get overrideAccScale() -> (overrideAccScale:Vector)
   ) -> (Pose) {
    bClampLinearTranslationLimitToRefPose=true
    bOverrideWorldGravity=true
    overridePhysicsAsset=&/F_MED_Raider_Silver_Pack_Physics
    simulationSpace=BaseBoneSpace
    alphaBoolBlend {blendOption=Linear}
    baseBoneRef {boneName=root}
    componentAppliedLinearAccClamp=<1000 1000 1000>
    componentLinearAccScale=<0.35 0.3 0.3>
    componentLinearVelScale=<0.2 0.2 0.2>
    overrideWorldGravity {z=-980}
   }
   n2 = Get isSkydiving() -> (isSkydiving:double)
   n3 = "Transform (Modify) Bone - Bone: dyn_R_grenade_jnt"(ComponentPose=@n1, Alpha=@n2, Translation="0,1.5,0", Rotation="0.000,20,0.000", Scale="1,1,1") {
    rotationMode=BMM_Additive
    rotationSpace=BCS_BoneSpace
    translationMode=BMM_Additive
    translationSpace=BCS_BoneSpace
    alphaBoolBlend {blendOption=Linear}
    alphaScaleBiasClamp {bInterpResult=true, interpSpeedIncreasing=50}
    boneToModify {boneName=dyn_R_grenade_jnt}
    rotation {yaw=20}
    translation {y=1.5}
   }
   n4 = "Transform (Modify) Bone - Bone: dyn_L_grenade_jnt"(ComponentPose=@n3, Alpha=@n2, Translation="0,-1.7,0", Rotation="0.000,-20,0.000", Scale="1,1,1") {
    rotationMode=BMM_Additive
    rotationSpace=BCS_BoneSpace
    translationMode=BMM_Additive
    translationSpace=BCS_BoneSpace
    alphaBoolBlend {blendOption=Linear}
    alphaScaleBiasClamp {bInterpResult=true, interpSpeedIncreasing=50}
    boneToModify {boneName=dyn_L_grenade_jnt}
    rotation {yaw=-20}
    translation {y=-1.7}
   }
   Output Pose(Result=Component To Local(ComponentPose=@n4)) {layerGroup=None}
  ]}
 EventGraph {nodes [
   Event BlueprintUpdateAnimation()
   > n1 = Sequence!(execute) -> (then_0, then_1, then_2:Exec)
   > n2 = Branch(Condition=Get bIsInFrontEnd())
   > Set overrideAccScale(overrideAccScale:Vector="0.05,0.05,0.05") -> (Output_Get:Vector)
   Set overrideAccScale(execute=@n2.else, overrideAccScale="0.35,0.3,0.3")
   Set isSkydiving(execute=@n1.then_1, isSkydiving:double="To Float (Boolean)"(InBool=Get bIsSkydiveDiveMode())) -> (Output_Get:double)
   > n3 = Branch(Condition=Get bIsSkyDiving())
   > n4 = Branch(Condition=Get bIsParachuteOpened())
   Set gravity_override(execute=[@n3.else, @n4.then], gravity_override="0, 0,-980")
   Set gravity_override(execute=@n4.else, gravity_override="0, 0,-100")
   TryGetPawnOwner()
  ]}
]
variables [
 gravity_override {type=Vector}
 isSkydiving {type=number}
 overrideAccScale {type=Vector, default=<0.35 0.3 0.3>}
]
